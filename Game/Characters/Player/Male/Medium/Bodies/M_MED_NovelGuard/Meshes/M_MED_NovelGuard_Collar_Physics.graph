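# Physics asset for the collar of the medium male NovelGuard player body, defining the bodies and joints for its dynamic collar bones.
# Capsules on the head, clavicles and spine_05 form unconstrained body volumes around the collar.
# Left, right and back flaps each hinge from a small MSTR anchor body, while the two front pieces are two-link chains from their anchors.
# Angular limits keep twist and swing1 tight, leaving swing2 as the main bend axis.
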PhysicsAsset /Game/Characters/Player/Male/Medium/Bodies/M_MED_NovelGuard/Meshes/M_MED_NovelGuard_Collar_Physics
bodies {
 clavicle_l [{shapeType=Capsule, length=8, radius=9, shapeName=clavicle_l_capsule, center=<12.645 -0.0001 -0.252>, rotation=<rot -70 179.9984 -179.9987>}]
 clavicle_r [{
   shapeType=Capsule
   length=6.5
   radius=10.5
   shapeName=clavicle_r_capsule
   center=<-14.5818 0.0001 1.3387>
   rotation=<rot -70 179.998 -179.9983>
  }]
 dyn_collar_01_L [{
   shapeType=Capsule
   length=16.5
   radius=1.5
   shapeName=dyn_collar_01_L_capsule
   center=<4.8556 0.6341 0.3147>
   rotation=<rot 0.0005 -19.9994 74.9997>
  }]
 dyn_collar_01_R [{
   shapeType=Capsule
   length=17
   radius=1.5
   shapeName=dyn_collar_01_R_capsule
   center=<4.5452 0.1709 -0.0879>
   rotation=<rot 0.0004 24.9996 115.0002>
  }]
 dyn_collar_MSTR_L [{shapeType=Capsule, length=0, radius=0.5, shapeName=dyn_collar_MSTR_L_capsule}]
 dyn_collar_MSTR_R [{shapeType=Capsule, length=0, radius=0.5, shapeName=dyn_collar_MSTR_R_capsule}]
 dyn_collar_bk_01 [{shapeType=Capsule, length=12.5, radius=3, shapeName=dyn_collar_bk_01_capsule, center=<5.7068 0 0>, rotation=<rot 0.0004 0 90>}]
 dyn_collar_bk_MSTR [{shapeType=Capsule, length=0, radius=0.5, shapeName=dyn_collar_bk_MSTR_capsule}]
 dyn_collar_fr_01_L [{shapeType=Capsule, length=4.0083, radius=1, shapeName=dyn_collar_fr_01_L_capsule, center=<0.778 0 0>, rotation=<rot 90 -179.8417 -179.8421>}]
 dyn_collar_fr_01_R [{shapeType=Capsule, length=4.0083, radius=1, shapeName=dyn_collar_fr_01_L_capsule, center=<0.778 0 0>, rotation=<rot 90 -179.8417 -179.8421>}]
 dyn_collar_fr_02_L [{
   shapeType=Capsule
   length=11.2897
   radius=4.0045
   shapeName=dyn_collar_fr_02_L_capsule
   center=<-2.7 -1.7 -1.7>
   rotation=<rot -80 179.9999 -179.9999>
  }]
 dyn_collar_fr_02_R [{
   shapeType=Capsule
   length=8.3993
   radius=3.9033
   shapeName=dyn_collar_fr_02_R_capsule
   center=<3.7256 -1.3375 -2.2447>
   rotation=<rot -5.9012 73.6123 -119.7299>
  }]
 dyn_collar_fr_MSTR_L [{shapeType=Capsule, length=0, radius=0.5, shapeName=dyn_collar_fr_MSTR_L_capsule}]
 dyn_collar_fr_MSTR_R [{shapeType=Capsule, length=0, radius=0.5, shapeName=dyn_collar_fr_MSTR_R_capsule}]
 head [{shapeType=Capsule, length=6, radius=7, shapeName=head_capsule, center=<1.7172 0 0>, rotation=<rot -90 0 -0.0002>}]
 spine_05 [
  {shapeType=Capsule, length=7, radius=12, shapeName=spine_05_capsule, rotation=<rot 0 -0.0003 0>}
  {shapeType=Capsule, length=16, radius=8.5, shapeName=spine_05_capsule, center=<-4.5533 4.606 7.806>, rotation=<rot -70 -0.0005 0>}
 ]
}
constraints [
 {bone1Name=dyn_collar_fr_02_L, bone2Name=dyn_collar_fr_01_L, swing1LimitDegrees=2.5, swing2LimitDegrees=35, twistLimitDegrees=2.5}
 {bone1Name=dyn_collar_fr_01_L, bone2Name=dyn_collar_fr_MSTR_L, swing1LimitDegrees=2.5, swing2LimitDegrees=45, twistLimitDegrees=2.5}
 {bone1Name=dyn_collar_fr_01_R, bone2Name=dyn_collar_fr_MSTR_R, swing1LimitDegrees=2.5, swing2LimitDegrees=45, twistLimitDegrees=2.5}
 {bone1Name=dyn_collar_fr_02_R, bone2Name=dyn_collar_fr_01_R, swing1LimitDegrees=2.5, swing2LimitDegrees=45, twistLimitDegrees=2.5}
 {bone1Name=dyn_collar_bk_01, bone2Name=dyn_collar_bk_MSTR, swing1LimitDegrees=5, swing2LimitDegrees=25, twistLimitDegrees=5}
 {bone1Name=dyn_collar_01_R, bone2Name=dyn_collar_MSTR_R, swing1LimitDegrees=5, swing2LimitDegrees=45, twistLimitDegrees=5}
 {bone1Name=dyn_collar_01_L, bone2Name=dyn_collar_MSTR_L, swing1LimitDegrees=5, swing2LimitDegrees=45, twistLimitDegrees=15}
]
