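# Physics asset for the male medium NovelGuard player body, covering its dynamic coat and sword plus pelvis, belt and legs.
# Five coat chains per side, three segments left and four right, hang from tiny master bodies freely jointed to the belt, with swing2 limits mostly tightening downward.
# Extra constraints tie some neighboring chains sideways.
# The two-segment sword chain mainly swings on one axis, while leg joints are unrestricted.
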PhysicsAsset /Game/Characters/Player/Male/Medium/Bodies/M_MED_NovelGuard/Meshes/M_MED_NovelGuard_Physics
bodies {
 belt [{shapeType=Capsule, length=11, radius=6, shapeName=belt_capsule}]
 calf_l [{shapeType=Capsule, length=30.5, radius=11, shapeName=calf_l_capsule, center=<-22.0335 0 0>, rotation=<rot -90 -179.9605 179.9604>}]
 calf_r [{shapeType=Capsule, length=32, radius=10, shapeName=calf_r_capsule, center=<21.5731 2.6552 -1.507>, rotation=<rot -90 -179.9983 179.9979>}]
 dyn_coat_01_L [{shapeType=Capsule, length=17.5, radius=5, shapeName=dyn_coat_bk_01_L_capsule, center=<8.4782 0 0>, rotation=<rot -90 -159.444 69.4443>}]
 dyn_coat_01_R [{
   shapeType=Capsule
   length=16.9375
   radius=6.5938
   shapeName=dyn_coat_bk_01_R_capsule
   center=<8.4782 0 0>
   rotation=<rot -90 -95.0322 5.0316>
  }]
 dyn_coat_02_L [{shapeType=Capsule, length=17.5, radius=5, shapeName=dyn_coat_fr_02_L_capsule, center=<8.4782 0 0>, rotation=<rot -90 180 90.0002>}]
 dyn_coat_02_R [{shapeType=Capsule, length=20.8438, radius=7.6875, shapeName=dyn_coat_bk_02_R_capsule, center=<8.4782 0 0>, rotation=<rot -90 180 89.9996>}]
 dyn_coat_03_L [{shapeType=Capsule, length=17.5, radius=6, shapeName=dyn_coat_bk_03_L_capsule, center=<8.4782 0 0>, rotation=<rot -90 -159.444 69.4443>}]
 dyn_coat_03_R [{
   shapeType=Capsule
   length=18.8438
   radius=7.9375
   shapeName=dyn_coat_bk_03_R_capsule
   center=<8.4782 -0.0001 0>
   rotation=<rot -90 165.9638 104.0358>
  }]
 dyn_coat_04_R [{
   shapeType=Capsule
   length=34.2188
   radius=9
   shapeName=dyn_coat_bk_04_R_capsule
   center=<8.4782 -0.4464 0>
   rotation=<rot -90 -172.875 77.8745>
  }]
 dyn_coat_MSTR_L [{shapeType=Capsule, length=0, radius=0.5, shapeName=dyn_coat_bk_MSTR_L_capsule}]
 dyn_coat_MSTR_R [{shapeType=Capsule, length=0, radius=0.5, shapeName=dyn_coat_bk_MSTR_R_capsule}]
 dyn_coat_bk_01_L [{shapeType=Capsule, length=17.5, radius=5, shapeName=dyn_coat_bk_01_L_capsule, center=<8.4782 0 0>, rotation=<rot -90 -159.444 69.4443>}]
 dyn_coat_bk_01_R [{shapeType=Capsule, length=17.5, radius=5, shapeName=dyn_coat_bk_01_R_capsule, center=<8.4782 0 0>, rotation=<rot -90 180 89.9996>}]
 dyn_coat_bk_02_L [{shapeType=Capsule, length=17.5, radius=5, shapeName=dyn_coat_bk_02_L_capsule, center=<8.4782 0 0>, rotation=<rot -90 -159.444 69.4443>}]
 dyn_coat_bk_02_R [{shapeType=Capsule, length=17.5, radius=6, shapeName=dyn_coat_bk_02_R_capsule, center=<11.7222 -0.0001 0>, rotation=<rot -90 180 89.9996>}]
 dyn_coat_bk_03_L [{shapeType=Capsule, length=17.5, radius=6, shapeName=dyn_coat_bk_03_L_capsule, center=<8.4782 0 0>, rotation=<rot -90 -159.444 69.4443>}]
 dyn_coat_bk_03_R [{
   shapeType=Capsule
   length=15.4375
   radius=9.0625
   shapeName=dyn_coat_bk_03_R_capsule
   center=<10.5093 -0.0001 0>
   rotation=<rot -90 165.9638 104.0358>
  }]
 dyn_coat_bk_04_R [{
   shapeType=Capsule
   length=32.375
   radius=9.0625
   shapeName=dyn_coat_bk_04_R_capsule
   center=<13.6493 -0.0002 0>
   rotation=<rot -90 165.9638 104.0358>
  }]
 dyn_coat_bk_MSTR_L [{shapeType=Capsule, length=0, radius=0.5, shapeName=dyn_coat_bk_MSTR_L_capsule}]
 dyn_coat_bk_MSTR_R [{shapeType=Capsule, length=0, radius=0.5, shapeName=dyn_coat_bk_MSTR_R_capsule}]
 dyn_coat_bk_mid_01_L [{shapeType=Capsule, length=17.5, radius=5, shapeName=dyn_coat_bk_01_L_capsule, center=<8.4782 0 0>, rotation=<rot -90 -159.444 69.4443>}]
 dyn_coat_bk_mid_01_R [{shapeType=Capsule, length=23.9375, radius=6.5938, shapeName=dyn_coat_bk_01_R_capsule, center=<8.4782 0 0>, rotation=<rot -90 180 89.9996>}]
 dyn_coat_bk_mid_02_L [{shapeType=Capsule, length=17.5, radius=5, shapeName=dyn_coat_fr_02_L_capsule, center=<8.4782 0 0>, rotation=<rot -90 180 90.0002>}]
 dyn_coat_bk_mid_02_R [{shapeType=Capsule, length=22.8438, radius=8.6875, shapeName=dyn_coat_bk_02_R_capsule, center=<8.4782 0 0>, rotation=<rot -90 180 89.9996>}]
 dyn_coat_bk_mid_03_L [{shapeType=Capsule, length=17.5, radius=6, shapeName=dyn_coat_bk_03_L_capsule, center=<8.4782 0 0>, rotation=<rot -90 -159.444 69.4443>}]
 dyn_coat_bk_mid_03_R [{
   shapeType=Capsule
   length=17.9375
   radius=9.0625
   shapeName=dyn_coat_bk_03_R_capsule
   center=<8.4782 -0.0001 0>
   rotation=<rot -90 165.9638 104.0358>
  }]
 dyn_coat_bk_mid_04_R [{
   shapeType=Capsule
   length=32.375
   radius=9.0625
   shapeName=dyn_coat_bk_04_R_capsule
   center=<8.4782 -0.0002 0>
   rotation=<rot -90 165.9638 104.0358>
  }]
 dyn_coat_bk_mid_MSTR_L [{shapeType=Capsule, length=0, radius=0.5, shapeName=dyn_coat_bk_MSTR_L_capsule}]
 dyn_coat_bk_mid_MSTR_R [{shapeType=Capsule, length=0, radius=0.5, shapeName=dyn_coat_bk_MSTR_R_capsule}]
 dyn_coat_fr_01_L [{shapeType=Capsule, length=17.5, radius=5, shapeName=dyn_coat_fr_01_L_capsule, center=<8.4782 0 0>, rotation=<rot -90 180 90.0002>}]
 dyn_coat_fr_01_R [{shapeType=Capsule, length=17.5, radius=5, shapeName=dyn_coat_fr_01_R_capsule, center=<8.4782 0 0>, rotation=<rot -90 165.9638 104.0359>}]
 dyn_coat_fr_02_L [{shapeType=Capsule, length=17.5, radius=5, shapeName=dyn_coat_fr_02_L_capsule, center=<8.4782 0 0>, rotation=<rot -90 180 90.0002>}]
 dyn_coat_fr_02_R [{shapeType=Capsule, length=17.5, radius=5, shapeName=dyn_coat_fr_02_R_capsule, center=<8.4782 0 0>, rotation=<rot -90 165.9638 104.0359>}]
 dyn_coat_fr_03_L [{shapeType=Capsule, length=17.5, radius=6, shapeName=dyn_coat_fr_03_L_capsule, center=<8.4782 0 0>, rotation=<rot -90 180 90.0002>}]
 dyn_coat_fr_03_R [{
   shapeType=Capsule
   length=15.4375
   radius=9.0625
   shapeName=dyn_coat_fr_03_R_capsule
   center=<8.4782 0 0>
   rotation=<rot -90 165.9638 104.0359>
  }]
 dyn_coat_fr_04_R [{
   shapeType=Capsule
   length=32.375
   radius=9.0625
   shapeName=dyn_coat_fr_04_R_capsule
   center=<8.4782 0 0>
   rotation=<rot -90 165.9638 104.0359>
  }]
 dyn_coat_fr_MSTR_L [{shapeType=Capsule, length=0, radius=0.5, shapeName=dyn_coat_fr_MSTR_L_capsule}]
 dyn_coat_fr_MSTR_R [{shapeType=Capsule, length=0, radius=0.5, shapeName=dyn_coat_fr_MSTR_R_capsule}]
 dyn_coat_fr_mid_01_L [{shapeType=Capsule, length=17.5, radius=5, shapeName=dyn_coat_fr_01_L_capsule, center=<8.4782 0 0>, rotation=<rot -90 180 90.0002>}]
 dyn_coat_fr_mid_01_R [{
   shapeType=Capsule
   length=25.4375
   radius=6.5938
   shapeName=dyn_coat_fr_01_R_capsule
   center=<8.4782 0 0>
   rotation=<rot -90 165.9638 104.0359>
  }]
 dyn_coat_fr_mid_02_L [{shapeType=Capsule, length=17.5, radius=5, shapeName=dyn_coat_fr_02_L_capsule, center=<8.4782 0 0>, rotation=<rot -90 180 90.0002>}]
 dyn_coat_fr_mid_02_R [{
   shapeType=Capsule
   length=22.8438
   radius=7.6875
   shapeName=dyn_coat_fr_02_R_capsule
   center=<8.4782 0 0>
   rotation=<rot -90 165.9638 104.0359>
  }]
 dyn_coat_fr_mid_03_L [{shapeType=Capsule, length=17.5, radius=6, shapeName=dyn_coat_fr_03_L_capsule, center=<8.4782 0 0>, rotation=<rot -90 180 90.0002>}]
 dyn_coat_fr_mid_03_R [{
   shapeType=Capsule
   length=15.4375
   radius=9.0625
   shapeName=dyn_coat_fr_03_R_capsule
   center=<8.4782 0 0>
   rotation=<rot -90 165.9638 104.0359>
  }]
 dyn_coat_fr_mid_04_R [{
   shapeType=Capsule
   length=15.4375
   radius=9.0625
   shapeName=dyn_coat_fr_04_R_capsule
   center=<8.4782 0 0>
   rotation=<rot -90 165.9638 104.0359>
  }]
 dyn_coat_fr_mid_MSTR_L [{shapeType=Capsule, length=0, radius=0.5, shapeName=dyn_coat_fr_MSTR_L_capsule}]
 dyn_coat_fr_mid_MSTR_R [{shapeType=Capsule, length=0, radius=0.5, shapeName=dyn_coat_fr_MSTR_R_capsule}]
 dyn_sword_01 [{
   shapeType=Box
   extentX=20.172
   extentY=7.7
   extentZ=19.3061
   shapeName=dyn_sword_01_box
   center=<11.6994 6.5584 -0.049>
   rotation=<rot -1.0879 -169.6271 -93.8178>
  }]
 dyn_sword_02 [{
   shapeType=Capsule
   length=83.5312
   radius=7.4688
   shapeName=dyn_sword_02_capsule
   center=<16.1749 -0.4025 0>
   rotation=<rot -90 -10.6197 8.284>
  }]
 dyn_sword_MSTR [{shapeType=Capsule, length=0, radius=0.5, shapeName=dyn_sword_MSTR_capsule}]
 foot_l [{shapeType=Capsule, length=20, radius=8.5, shapeName=foot_r_capsule, center=<-4.8716 5.334 0>, rotation=<rot -0.0001 -0.0008 90>}]
 foot_r [{shapeType=Capsule, length=20, radius=8.5, shapeName=foot_r_capsule, center=<6.6013 -5.4713 0>, rotation=<rot -0.0001 -0.0011 -90>}]
 pelvis [{shapeType=Capsule, length=0, radius=0.5, shapeName=pelvis_capsule}]
 thigh_l [{shapeType=Capsule, length=36, radius=7.5, shapeName=thigh_l_capsule, center=<-22.5291 0 2.5>, rotation=<rot -90 -179.8881 179.5356>}]
 thigh_r [{shapeType=Capsule, length=34, radius=9, shapeName=thigh_r_capsule, center=<26.6161 0 -4>, rotation=<rot -90 -179.8418 179.8412>}]
}
constraints [
 {bone1Name=dyn_coat_bk_01_L, bone2Name=dyn_coat_bk_MSTR_L, swing1LimitDegrees=5, swing2LimitDegrees=45, twistLimitDegrees=10}
 {bone1Name=dyn_coat_bk_02_L, bone2Name=dyn_coat_bk_01_L, swing1LimitDegrees=5, swing2LimitDegrees=35, twistLimitDegrees=10}
 {bone1Name=dyn_coat_bk_03_L, bone2Name=dyn_coat_bk_02_L, swing1LimitDegrees=5, swing2LimitDegrees=25, twistLimitDegrees=10}
 {bone1Name=dyn_coat_bk_01_R, bone2Name=dyn_coat_bk_MSTR_R, swing1LimitDegrees=5, swing2LimitDegrees=45, twistLimitDegrees=10}
 {bone1Name=dyn_coat_bk_02_R, bone2Name=dyn_coat_bk_01_R, swing1LimitDegrees=5, swing2LimitDegrees=35, twistLimitDegrees=10}
 {bone1Name=dyn_coat_bk_03_R, bone2Name=dyn_coat_bk_02_R, swing1LimitDegrees=10, swing2LimitDegrees=25, twistLimitDegrees=10}
 {bone1Name=dyn_coat_bk_04_R, bone2Name=dyn_coat_bk_03_R, swing1LimitDegrees=10, swing2LimitDegrees=15, twistLimitDegrees=10}
 {bone1Name=dyn_coat_fr_01_L, bone2Name=dyn_coat_fr_MSTR_L, swing1LimitDegrees=5, swing2LimitDegrees=45, twistLimitDegrees=10}
 {bone1Name=dyn_coat_fr_02_L, bone2Name=dyn_coat_fr_01_L, swing1LimitDegrees=5, swing2LimitDegrees=35, twistLimitDegrees=10}
 {bone1Name=dyn_coat_fr_03_L, bone2Name=dyn_coat_fr_02_L, swing1LimitDegrees=5, swing2LimitDegrees=25, twistLimitDegrees=10}
 {bone1Name=dyn_coat_fr_01_R, bone2Name=dyn_coat_fr_MSTR_R, swing1LimitDegrees=5, swing2LimitDegrees=45, twistLimitDegrees=10}
 {bone1Name=dyn_coat_fr_02_R, bone2Name=dyn_coat_fr_01_R, swing1LimitDegrees=5, swing2LimitDegrees=45, twistLimitDegrees=10}
 {bone1Name=dyn_coat_fr_03_R, bone2Name=dyn_coat_fr_02_R, swing1LimitDegrees=5, swing2LimitDegrees=25, twistLimitDegrees=10}
 {bone1Name=dyn_coat_fr_04_R, bone2Name=dyn_coat_fr_03_R, swing1LimitDegrees=5, swing2LimitDegrees=15, twistLimitDegrees=10}
 {
  bone1Name=foot_r
  bone2Name=calf_r
  swing1LimitDegrees=45
  swing1Motion=Free
  swing2LimitDegrees=45
  swing2Motion=Free
  twistLimitDegrees=45
  twistMotion=Free
 }
 {bone1Name=dyn_coat_bk_mid_01_L, bone2Name=dyn_coat_bk_mid_MSTR_L, swing1LimitDegrees=10, swing2LimitDegrees=45, twistLimitDegrees=10}
 {bone1Name=dyn_coat_bk_mid_02_L, bone2Name=dyn_coat_bk_mid_01_L, swing1LimitDegrees=10, swing2LimitDegrees=35, twistLimitDegrees=10}
 {bone1Name=dyn_coat_bk_mid_03_L, bone2Name=dyn_coat_bk_mid_02_L, swing1LimitDegrees=10, swing2LimitDegrees=25, twistLimitDegrees=10}
 {bone1Name=dyn_coat_bk_mid_01_R, bone2Name=dyn_coat_bk_mid_MSTR_R, swing1LimitDegrees=10, swing2LimitDegrees=45, twistLimitDegrees=10}
 {bone1Name=dyn_coat_bk_mid_02_R, bone2Name=dyn_coat_bk_mid_01_R, swing1LimitDegrees=10, swing2LimitDegrees=35, twistLimitDegrees=10}
 {bone1Name=dyn_coat_bk_mid_03_R, bone2Name=dyn_coat_bk_mid_02_R, swing1LimitDegrees=15, swing2LimitDegrees=25, twistLimitDegrees=15}
 {bone1Name=dyn_coat_bk_mid_04_R, bone2Name=dyn_coat_bk_mid_03_R, swing1LimitDegrees=10, swing2LimitDegrees=15, twistLimitDegrees=10}
 {bone1Name=dyn_coat_fr_mid_01_L, bone2Name=dyn_coat_fr_mid_MSTR_L, swing1LimitDegrees=10, swing2LimitDegrees=45, twistLimitDegrees=10}
 {bone1Name=dyn_coat_fr_mid_02_L, bone2Name=dyn_coat_fr_mid_01_L, swing1LimitDegrees=10, swing2LimitDegrees=35, twistLimitDegrees=10}
 {bone1Name=dyn_coat_fr_mid_03_L, bone2Name=dyn_coat_fr_mid_02_L, swing1LimitDegrees=10, swing2LimitDegrees=25, twistLimitDegrees=10}
 {bone1Name=dyn_coat_fr_mid_01_R, bone2Name=dyn_coat_fr_mid_MSTR_R, swing1LimitDegrees=10, swing2LimitDegrees=45, twistLimitDegrees=10}
 {bone1Name=dyn_coat_fr_mid_02_R, bone2Name=dyn_coat_fr_mid_01_R, swing1LimitDegrees=10, swing2LimitDegrees=45, twistLimitDegrees=10}
 {bone1Name=dyn_coat_fr_mid_03_R, bone2Name=dyn_coat_fr_mid_02_R, swing1LimitDegrees=10, swing2LimitDegrees=25, twistLimitDegrees=10}
 {bone1Name=dyn_coat_fr_mid_04_R, bone2Name=dyn_coat_fr_mid_03_R, swing1LimitDegrees=10, swing2LimitDegrees=15, twistLimitDegrees=10}
 {bone1Name=dyn_coat_01_L, bone2Name=dyn_coat_MSTR_L, swing1LimitDegrees=10, swing2LimitDegrees=45, twistLimitDegrees=10}
 {bone1Name=dyn_coat_02_L, bone2Name=dyn_coat_01_L, swing1LimitDegrees=10, swing2LimitDegrees=35, twistLimitDegrees=10}
 {bone1Name=dyn_coat_03_L, bone2Name=dyn_coat_02_L, swing1LimitDegrees=10, swing2LimitDegrees=25, twistLimitDegrees=10}
 {bone1Name=dyn_coat_01_R, bone2Name=dyn_coat_MSTR_R, swing1LimitDegrees=10, swing2LimitDegrees=45, twistLimitDegrees=10}
 {bone1Name=dyn_coat_02_R, bone2Name=dyn_coat_01_R, swing1LimitDegrees=10, swing2LimitDegrees=35, twistLimitDegrees=10}
 {bone1Name=dyn_coat_03_R, bone2Name=dyn_coat_02_R, swing1LimitDegrees=10, swing2LimitDegrees=25, twistLimitDegrees=10}
 {bone1Name=dyn_coat_04_R, bone2Name=dyn_coat_03_R, swing1LimitDegrees=10, swing2LimitDegrees=15, twistLimitDegrees=10}
 {bone1Name=dyn_coat_bk_mid_02_L, bone2Name=dyn_coat_02_L, swing1LimitDegrees=20, swing2LimitDegrees=45, twistLimitDegrees=20}
 {bone1Name=dyn_coat_bk_mid_02_R, bone2Name=dyn_coat_02_R, swing1LimitDegrees=25, swing2LimitDegrees=25, twistLimitDegrees=25}
 {
  bone1Name=dyn_coat_bk_02_L
  bone2Name=dyn_coat_02_L
  swing1LimitDegrees=45
  swing1Motion=Free
  swing2LimitDegrees=45
  swing2Motion=Free
  twistLimitDegrees=45
  twistMotion=Free
 }
 {
  bone1Name=foot_l
  bone2Name=calf_l
  swing1LimitDegrees=45
  swing1Motion=Free
  swing2LimitDegrees=45
  swing2Motion=Free
  twistLimitDegrees=45
  twistMotion=Free
 }
 {bone1Name=dyn_coat_03_R, bone2Name=dyn_coat_fr_mid_03_R, swing1LimitDegrees=45, swing2LimitDegrees=25, twistLimitDegrees=45}
 {bone1Name=dyn_coat_fr_mid_02_R, bone2Name=dyn_coat_fr_02_R, swing1LimitDegrees=45, swing2LimitDegrees=55, twistLimitDegrees=45}
 {
  bone1Name=dyn_coat_bk_mid_02_R
  bone2Name=dyn_coat_bk_02_R
  swing1LimitDegrees=45
  swing1Motion=Free
  swing2LimitDegrees=45
  swing2Motion=Free
  twistLimitDegrees=45
  twistMotion=Free
 }
 {bone1Name=dyn_coat_bk_mid_03_R, bone2Name=dyn_coat_bk_03_R, swing1LimitDegrees=45, swing2LimitDegrees=25, twistLimitDegrees=45}
 {bone1Name=dyn_sword_02, bone2Name=dyn_sword_01, swing1LimitDegrees=25, swing2LimitDegrees=2, twistLimitDegrees=2}
 {
  bone1Name=dyn_sword_01
  bone2Name=dyn_sword_MSTR
  swing1LimitDegrees=5
  swing1Motion=Locked
  swing2LimitDegrees=55
  twistLimitDegrees=5
  twistMotion=Locked
 }
 {
  bone1Name=calf_l
  bone2Name=thigh_l
  swing1LimitDegrees=45
  swing1Motion=Free
  swing2LimitDegrees=45
  swing2Motion=Free
  twistLimitDegrees=45
  twistMotion=Free
 }
 {
  bone1Name=calf_r
  bone2Name=thigh_r
  swing1LimitDegrees=45
  swing1Motion=Free
  swing2LimitDegrees=45
  swing2Motion=Free
  twistLimitDegrees=45
  twistMotion=Free
 }
 {
  bone1Name=thigh_l
  bone2Name=pelvis
  swing1LimitDegrees=45
  swing1Motion=Free
  swing2LimitDegrees=45
  swing2Motion=Free
  twistLimitDegrees=45
  twistMotion=Free
 }
 {
  bone1Name=thigh_r
  bone2Name=pelvis
  swing1LimitDegrees=45
  swing1Motion=Free
  swing2LimitDegrees=45
  swing2Motion=Free
  twistLimitDegrees=45
  twistMotion=Free
 }
 {
  bone1Name=dyn_sword_MSTR
  bone2Name=belt
  swing1LimitDegrees=45
  swing1Motion=Free
  swing2LimitDegrees=45
  swing2Motion=Free
  twistLimitDegrees=45
  twistMotion=Free
 }
 {
  bone1Name=dyn_coat_MSTR_L
  bone2Name=belt
  swing1LimitDegrees=45
  swing1Motion=Free
  swing2LimitDegrees=45
  swing2Motion=Free
  twistLimitDegrees=45
  twistMotion=Free
 }
 {
  bone1Name=dyn_coat_MSTR_R
  bone2Name=belt
  swing1LimitDegrees=45
  swing1Motion=Free
  swing2LimitDegrees=45
  swing2Motion=Free
  twistLimitDegrees=45
  twistMotion=Free
 }
 {
  bone1Name=dyn_coat_bk_MSTR_L
  bone2Name=belt
  swing1LimitDegrees=45
  swing1Motion=Free
  swing2LimitDegrees=45
  swing2Motion=Free
  twistLimitDegrees=45
  twistMotion=Free
 }
 {
  bone1Name=dyn_coat_bk_MSTR_R
  bone2Name=belt
  swing1LimitDegrees=45
  swing1Motion=Free
  swing2LimitDegrees=45
  swing2Motion=Free
  twistLimitDegrees=45
  twistMotion=Free
 }
 {
  bone1Name=dyn_coat_bk_mid_MSTR_L
  bone2Name=belt
  swing1LimitDegrees=45
  swing1Motion=Free
  swing2LimitDegrees=45
  swing2Motion=Free
  twistLimitDegrees=45
  twistMotion=Free
 }
 {
  bone1Name=dyn_coat_bk_mid_MSTR_R
  bone2Name=belt
  swing1LimitDegrees=45
  swing1Motion=Free
  swing2LimitDegrees=45
  swing2Motion=Free
  twistLimitDegrees=45
  twistMotion=Free
 }
 {
  bone1Name=dyn_coat_fr_MSTR_L
  bone2Name=belt
  swing1LimitDegrees=45
  swing1Motion=Free
  swing2LimitDegrees=45
  swing2Motion=Free
  twistLimitDegrees=45
  twistMotion=Free
 }
 {
  bone1Name=dyn_coat_fr_MSTR_R
  bone2Name=belt
  swing1LimitDegrees=45
  swing1Motion=Free
  swing2LimitDegrees=45
  swing2Motion=Free
  twistLimitDegrees=45
  twistMotion=Free
 }
 {
  bone1Name=dyn_coat_fr_mid_MSTR_L
  bone2Name=belt
  swing1LimitDegrees=45
  swing1Motion=Free
  swing2LimitDegrees=45
  swing2Motion=Free
  twistLimitDegrees=45
  twistMotion=Free
 }
 {
  bone1Name=dyn_coat_fr_mid_MSTR_R
  bone2Name=belt
  swing1LimitDegrees=45
  swing1Motion=Free
  swing2LimitDegrees=45
  swing2Motion=Free
  twistLimitDegrees=45
  twistMotion=Free
 }
 {
  bone1Name=belt
  bone2Name=pelvis
  swing1LimitDegrees=45
  swing1Motion=Free
  swing2LimitDegrees=45
  swing2Motion=Free
  twistLimitDegrees=45
  twistMotion=Free
 }
]
