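# Post-process style AnimBP for the NovelGuard player body, animating coat, collar, strap, armor, belt and sword over the pose copied from the parent mesh.
# Control Rig and state- or joint-driven bone offsets feed crouch, front-end or default RigidBody sims, then AnimDynamics on collar and strap.
# With rigid bodies disabled, output falls back to the pre-offset pose plus a sword constraint.
# A material parameter hides an element while the matching pickaxe is equipped outside skydiving.
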AnimBlueprint /Game/Characters/Player/Male/Medium/Bodies/M_MED_NovelGuard/Meshes/M_MED_NovelGuard_AnimBP
target_skeleton=&/M_MED_NovelGuard_Skeleton
graphs [
 AnimGraph {nodes [
   "Transform (Modify) Bone - Bone: pad_01"!(ComponentPose, Alpha=Get upperarm r rotation y() -> (upperarm r rotation y:double), Translation="0,0,0", Rotation="-17,0,0") -> (Pose) {
    rotationMode=BMM_Additive
    rotationSpace=BCS_BoneSpace
    translationMode=BMM_Additive
    translationSpace=BCS_BoneSpace
    alphaBoolBlend {blendOption=Linear}
    alphaScaleBiasClamp {bInterpResult=true, interpSpeedDecreasing=5, interpSpeedIncreasing=5}
    boneToModify {boneName=pad_01}
    rotation {pitch=-17}
   }
   RigidBody!(ComponentPose, Alpha, OverrideWorldGravity:Vector=Get Gravity Override()) -> (Pose) {
    bClampLinearTranslationLimitToRefPose=true
    bOverrideWorldGravity=true
    overridePhysicsAsset=&/M_MED_NovelGuard_Collar_Physics
    simulationSpace=BaseBoneSpace
    alphaBoolBlend {blendOption=Linear}
    baseBoneRef {boneName=spine_05}
    componentAppliedLinearAccClamp=<3000 3000 3000>
    componentLinearAccScale=<0.35 0.35 0.35>
    componentLinearVelScale=<0.35 0.35 0.35>
   }
   n1 = "Constraint - dyn_sword_02 (PPP)"!(
    ComponentPose=Local To Component(LocalPose=Use cached pose 'main cache'())
    Alpha=1
    ConstraintWeights_0=0.8234
    ConstraintWeights_1:float=0
    ConstraintWeights_2:float=0.1766
   ) -> (Pose) {
    alphaBoolBlend {blendOption=Linear}
    boneToModify {boneName=dyn_sword_02}
    constraintSetup [{targetBone {boneName=pelvis}}, {targetBone {boneName=thigh_r}}, {targetBone {boneName=thigh_twist_01_r}}]
    constraintWeights [1,1,1]
   }
   n2 = RigidBody!(
    ComponentPose=Local To Component(LocalPose=Use cached pose 'presim cache'())
    Alpha
    OverrideWorldGravity=Get Gravity Override()
    ComponentLinearAccScale:Vector=Get ComponentLinearAccScale() -> (ComponentLinearAccScale:Vector)
    ComponentLinearVelScale:Vector=Get ComponentLinearVelScale() -> (ComponentLinearVelScale:Vector)
   ) -> (Pose) {
    bClampLinearTranslationLimitToRefPose=true
    bOverrideWorldGravity=true
    overridePhysicsAsset=&/M_MED_NovelGuard_Physics
    simulationSpace=BaseBoneSpace
    alphaBoolBlend {blendOption=Linear}
    alphaScaleBiasClamp {interpSpeedDecreasing=5}
    baseBoneRef {boneName=pelvis}
    componentAppliedLinearAccClamp=<3000 3000 3000>
    componentLinearAccScale=<0.2 0.2 0.2>
    componentLinearVelScale=<0.2 0.2 0.2>
   }
   n3 = RigidBody!(ComponentPose=Local To Component(LocalPose=Use cached pose 'presim cache'()), Alpha, OverrideWorldGravity=Get Gravity Override()) -> (Pose) {
    bClampLinearTranslationLimitToRefPose=true
    bOverrideWorldGravity=true
    overridePhysicsAsset=&/M_MED_NovelGuard_Frontend_Physics
    simulationSpace=BaseBoneSpace
    alphaBoolBlend {blendOption=Linear}
    alphaScaleBiasClamp {interpSpeedDecreasing=5, interpSpeedIncreasing=5}
    baseBoneRef {boneName=pelvis}
    componentAppliedLinearAccClamp=<3000 3000 3000>
    componentLinearAccScale=<0.35 0.35 0.35>
    componentLinearVelScale=<0.35 0.35 0.35>
   }
   n4 = Blend Poses by bool(
    BlendPose_0=Component To Local(ComponentPose=@n3)
    BlendPose_1=Component To Local(ComponentPose=@n2)
    BlendTime_0=0.5
    BlendTime_1=0.5
    bActiveValue=Get bIsInFrontEnd()
   ) {blendType=Linear}
   n5 = RigidBody(ComponentPose=Local To Component(LocalPose=Use cached pose 'presim cache'()), OverrideWorldGravity=Get Gravity Override()) {
    bClampLinearTranslationLimitToRefPose=true
    bOverrideWorldGravity=true
    overridePhysicsAsset=&/M_MED_NovelGuard_Crouching_Physics
    simulationSpace=BaseBoneSpace
    alphaBoolBlend {blendOption=Linear}
    alphaScaleBiasClamp {interpSpeedIncreasing=25}
    baseBoneRef {boneName=pelvis}
    componentAppliedLinearAccClamp=<3000 3000 3000>
    componentLinearAccScale=<0.35 0.35 0.35>
    componentLinearVelScale=<0.35 0.35 0.35>
   }
   n6 = Component To Local(ComponentPose=@n5)
   n7 = Get bIsCrouching()
   n8 = Get Thigh bending when sprinting() -> (Thigh bending when sprinting:float)
   n9 = "Transform (Modify) Bone - Bone: dyn_coat_fr_MSTR_R"!(
    ComponentPose=Local To Component(LocalPose=Blend Poses by bool(BlendPose_0=@n6, BlendPose_1=@n4, BlendTime_1=0.3, bActiveValue=@n7) {blendType=Linear})
    Alpha=@n8
    Translation="-5,0,10"
    Rotation="5,0,0"
   ) -> (Pose) {
    rotationMode=BMM_Additive
    rotationSpace=BCS_BoneSpace
    translationMode=BMM_Additive
    translationSpace=BCS_BoneSpace
    alphaBoolBlend {blendOption=Linear}
    boneToModify {boneName=dyn_coat_fr_MSTR_R}
    rotation {pitch=5}
    translation {x=-5, z=10}
   }
   n10 = "Transform (Modify) Bone - Bone: dyn_coat_fr_mid_MSTR_R"!(ComponentPose=@n9, Alpha=@n8, Translation="-5,0,10", Rotation="5,0,0") -> (Pose) {
    rotationMode=BMM_Additive
    rotationSpace=BCS_BoneSpace
    translationMode=BMM_Additive
    translationSpace=BCS_BoneSpace
    alphaBoolBlend {blendOption=Linear}
    boneToModify {boneName=dyn_coat_fr_mid_MSTR_R}
    rotation {pitch=5}
    translation {x=-5, z=10}
   }
   n11 = "Transform (Modify) Bone - Bone: dyn_coat_fr_03_R"!(ComponentPose=@n10, Alpha=@n8, Rotation="-50,0,0") -> (Pose) {
    rotationMode=BMM_Additive
    rotationSpace=BCS_BoneSpace
    alphaBoolBlend {blendOption=Linear}
    alphaScaleBiasClamp {interpSpeedIncreasing=1.5}
    boneToModify {boneName=dyn_coat_fr_03_R}
    rotation {pitch=-50}
   }
   n12 = "Transform (Modify) Bone - Bone: dyn_coat_fr_mid_03_R"!(ComponentPose=@n11, Alpha=@n8, Rotation="-50,0,0") -> (Pose) {
    rotationMode=BMM_Additive
    rotationSpace=BCS_BoneSpace
    alphaBoolBlend {blendOption=Linear}
    alphaScaleBiasClamp {interpSpeedIncreasing=1.5}
    boneToModify {boneName=dyn_coat_fr_mid_03_R}
    rotation {pitch=-50}
   }
   n13 = Get moving() -> (moving:double)
   n14 = "Transform (Modify) Bone - Bone: dyn_coat_fr_MSTR_L"!(ComponentPose=@n12, Alpha=@n13, Rotation="30,0,0") -> (Pose) {
    rotationMode=BMM_Additive
    rotationSpace=BCS_BoneSpace
    alphaBoolBlend {blendOption=Linear}
    alphaScaleBiasClamp {bInterpResult=true, interpSpeedIncreasing=1.5}
    boneToModify {boneName=dyn_coat_fr_MSTR_L}
    rotation {pitch=30}
   }
   n15 = "Transform (Modify) Bone - Bone: dyn_coat_fr_MSTR_R"!(ComponentPose=@n14, Alpha=@n13, Rotation="20,0,0") -> (Pose) {
    rotationMode=BMM_Additive
    rotationSpace=BCS_BoneSpace
    alphaBoolBlend {blendOption=Linear}
    alphaScaleBiasClamp {bInterpResult=true, interpSpeedIncreasing=1.5}
    boneToModify {boneName=dyn_coat_fr_MSTR_R}
    rotation {pitch=20}
   }
   n16 = "Transform (Modify) Bone - Bone: dyn_coat_fr_mid_MSTR_L"!(ComponentPose=@n15, Alpha=@n13, Rotation="30,0,0") -> (Pose) {
    rotationMode=BMM_Additive
    rotationSpace=BCS_BoneSpace
    alphaBoolBlend {blendOption=Linear}
    alphaScaleBiasClamp {bInterpResult=true, interpSpeedIncreasing=1.5}
    boneToModify {boneName=dyn_coat_fr_mid_MSTR_L}
    rotation {pitch=30}
   }
   n17 = "Transform (Modify) Bone - Bone: dyn_coat_fr_mid_MSTR_R"!(ComponentPose=@n16, Alpha=@n13, Rotation="20,0,0") -> (Pose) {
    rotationMode=BMM_Additive
    rotationSpace=BCS_BoneSpace
    alphaBoolBlend {blendOption=Linear}
    alphaScaleBiasClamp {bInterpResult=true, interpSpeedIncreasing=1.5}
    boneToModify {boneName=dyn_coat_fr_mid_MSTR_R}
    rotation {pitch=20}
   }
   n18 = "Transform (Modify) Bone - Bone: dyn_coat_bk_MSTR_L"!(ComponentPose=@n17, Alpha=@n13, Rotation="10,10,0") -> (Pose) {
    rotationMode=BMM_Additive
    rotationSpace=BCS_BoneSpace
    alphaBoolBlend {blendOption=Linear}
    alphaScaleBiasClamp {bInterpResult=true, interpSpeedIncreasing=1.5}
    boneToModify {boneName=dyn_coat_bk_MSTR_L}
    rotation {pitch=10, yaw=10}
   }
   n19 = "Transform (Modify) Bone - Bone: dyn_coat_bk_MSTR_R"!(ComponentPose=@n18, Alpha=@n13, Rotation="10,-10,0") -> (Pose) {
    rotationMode=BMM_Additive
    rotationSpace=BCS_BoneSpace
    alphaBoolBlend {blendOption=Linear}
    alphaScaleBiasClamp {bInterpResult=true, interpSpeedIncreasing=1.5}
    boneToModify {boneName=dyn_coat_bk_MSTR_R}
    rotation {pitch=10, yaw=-10}
   }
   n20 = "Transform (Modify) Bone - Bone: dyn_coat_bk_mid_MSTR_L"!(ComponentPose=@n19, Alpha=@n13, Rotation="10,0,0") -> (Pose) {
    rotationMode=BMM_Additive
    rotationSpace=BCS_BoneSpace
    alphaBoolBlend {blendOption=Linear}
    alphaScaleBiasClamp {bInterpResult=true, interpSpeedIncreasing=1.5}
    boneToModify {boneName=dyn_coat_bk_mid_MSTR_L}
    rotation {pitch=10}
   }
   n21 = "Transform (Modify) Bone - Bone: dyn_coat_bk_mid_MSTR_R"!(ComponentPose=@n20, Alpha=@n13, Rotation="10,0,0") -> (Pose) {
    rotationMode=BMM_Additive
    rotationSpace=BCS_BoneSpace
    alphaBoolBlend {blendOption=Linear}
    alphaScaleBiasClamp {bInterpResult=true, interpSpeedIncreasing=1.5}
    boneToModify {boneName=dyn_coat_bk_mid_MSTR_R}
    rotation {pitch=10}
   }
   n22 = Get head pitch() -> (head pitch:float)
   n23 = "Transform (Modify) Bone - Bone: dyn_collar_bk_MSTR"!(ComponentPose=@n21, Alpha=@n22, Rotation="135,0,0") -> (Pose) {
    rotationMode=BMM_Additive
    rotationSpace=BCS_BoneSpace
    alphaBoolBlend {blendOption=Linear}
    boneToModify {boneName=dyn_collar_bk_MSTR}
    rotation {pitch=135}
   }
   n24 = "Transform (Modify) Bone - Bone: dyn_collar_MSTR_R"!(ComponentPose=@n23, Alpha=@n22, Rotation="125,0,0") -> (Pose) {
    rotationMode=BMM_Additive
    rotationSpace=BCS_BoneSpace
    alphaBoolBlend {blendOption=Linear}
    alphaScaleBiasClamp {interpSpeedDecreasing=5, interpSpeedIncreasing=5}
    boneToModify {boneName=dyn_collar_MSTR_R}
    rotation {pitch=125}
   }
   n25 = Anim Dynamics - Bone: dyn_collar_fr_01_L!(ComponentPose=@n24, Alpha=1, GravityOverride:Vector="0,0,-980") -> (Pose) {
    angularDampingOverride=1
    angularSpringConstant=20000
    bAngularSpring=true
    bOverrideAngularDamping=true
    bUseGravityOverride=true
    bUseSphericalLimits=true
    gravityScale=0.35
    simulationSpace=RootRelative
    alphaBoolBlend {blendOption=Linear}
    boundBone {boneName=dyn_collar_fr_01_L}
    componentLinearAccScale=<0.7 0.7 0.7>
    componentLinearVelScale=<0.7 0.7 0.7>
    gravityOverride {z=-980}
    physicsBodyDefinitions [{
      collisionType=CustomSphere
      sphereCollisionRadius=2
      boundBone {boneName=dyn_collar_fr_01_L}
      boxExtents=<30 40 40>
      constraintSetup {angularLimitsMax=<0 5 5>, angularLimitsMin=<0 -2 -5>, angularTarget=<1 0 0>}
      localJointOffset=<4 0 0>
     }]
    sphericalLimits [{limitRadius=10, drivingBone {boneName=clavicle_l}, sphereLocalOffset=<10 0 0>}]
   }
   n26 = Anim Dynamics - Bone: dyn_collar_fr_02_L!(ComponentPose=@n25, Alpha=1, GravityOverride:Vector="0,0,-980") -> (Pose) {
    angularDampingOverride=1
    angularSpringConstant=20000
    bAngularSpring=true
    bOverrideAngularDamping=true
    bUseGravityOverride=true
    bUseSphericalLimits=true
    gravityScale=0.35
    simulationSpace=RootRelative
    alphaBoolBlend {blendOption=Linear}
    boundBone {boneName=dyn_collar_fr_02_L}
    componentLinearAccScale=<0.7 0.7 0.7>
    componentLinearVelScale=<0.7 0.7 0.7>
    gravityOverride {z=-980}
    physicsBodyDefinitions [{
      collisionType=CustomSphere
      sphereCollisionRadius=2
      boundBone {boneName=dyn_collar_fr_02_L}
      boxExtents=<30 40 40>
      constraintSetup {angularLimitsMax=<0 5 5>, angularLimitsMin=<0 -2 -5>, angularTarget=<1 0 0>}
      localJointOffset=<4 0 0>
     }]
    sphericalLimits [{limitRadius=10, drivingBone {boneName=clavicle_l}, sphereLocalOffset=<10 0 0>}]
   }
   n27 = Anim Dynamics - Bone: dyn_collar_fr_01_R!(ComponentPose=@n26, Alpha=1, GravityOverride:Vector="0,0,-980") -> (Pose) {
    angularDampingOverride=1
    angularSpringConstant=20000
    bAngularSpring=true
    bOverrideAngularDamping=true
    bUseGravityOverride=true
    bUseSphericalLimits=true
    gravityScale=0.35
    simulationSpace=RootRelative
    alphaBoolBlend {blendOption=Linear}
    boundBone {boneName=dyn_collar_fr_01_R}
    componentLinearAccScale=<0.7 0.7 0.7>
    componentLinearVelScale=<0.7 0.7 0.7>
    gravityOverride {z=-980}
    physicsBodyDefinitions [{
      collisionType=CustomSphere
      sphereCollisionRadius=2
      boundBone {boneName=dyn_collar_fr_01_R}
      boxExtents=<25 35 35>
      constraintSetup {angularLimitsMax=<0 7 3>, angularLimitsMin=<0 0 -5>, angularTarget=<1 0 0>}
      localJointOffset=<4 0 -3>
     }]
    sphericalLimits [{limitRadius=10, drivingBone {boneName=clavicle_r}, sphereLocalOffset=<-8.8576 -4.5635 2.6615>}]
   }
   n28 = Anim Dynamics - Bone: dyn_collar_fr_02_R!(ComponentPose=@n27, Alpha=1, GravityOverride:Vector="0,0,-980") -> (Pose) {
    angularDampingOverride=1
    angularSpringConstant=20000
    bAngularSpring=true
    bOverrideAngularDamping=true
    bUseGravityOverride=true
    bUseSphericalLimits=true
    gravityScale=0.35
    simulationSpace=RootRelative
    alphaBoolBlend {blendOption=Linear}
    boundBone {boneName=dyn_collar_fr_02_R}
    componentLinearAccScale=<0.7 0.7 0.7>
    componentLinearVelScale=<0.7 0.7 0.7>
    gravityOverride {z=-980}
    physicsBodyDefinitions [{
      collisionType=CustomSphere
      sphereCollisionRadius=2
      boundBone {boneName=dyn_collar_fr_02_R}
      boxExtents=<25 35 35>
      constraintSetup {angularLimitsMax=<0 7 3>, angularLimitsMin=<0 0 -5>, angularTarget=<1 0 0>}
      localJointOffset=<4 0 -3>
     }]
    sphericalLimits [{limitRadius=10, drivingBone {boneName=clavicle_r}, sphereLocalOffset=<-9.578 -3.737 2.6615>}]
   }
   n29 = Anim Dynamics - Bone: dyn_collar_01_L!(ComponentPose=@n28, Alpha=1, GravityOverride:Vector="0,0,-980") -> (Pose) {
    angularDampingOverride=1
    angularSpringConstant=20000
    bAngularSpring=true
    bOverrideAngularDamping=true
    bUseGravityOverride=true
    gravityScale=0.35
    simulationSpace=RootRelative
    alphaBoolBlend {blendOption=Linear}
    boundBone {boneName=dyn_collar_01_L}
    componentLinearAccScale=<0.7 0.7 0.7>
    componentLinearVelScale=<0.7 0.7 0.7>
    gravityOverride {z=-980}
    physicsBodyDefinitions [{
      collisionType=CustomSphere
      sphereCollisionRadius=2
      boundBone {boneName=dyn_collar_01_L}
      boxExtents=<20 30 30>
      constraintSetup {angularLimitsMax=<0 5 0>, angularLimitsMin=<0 -3 0>, angularTarget=<1 0 0>}
      localJointOffset=<4 0 0>
     }]
   }
   n30 = Anim Dynamics - Bone: dyn_collar_bk_01!(ComponentPose=@n29, Alpha=1, GravityOverride:Vector="0,0,-980") -> (Pose) {
    angularDampingOverride=1
    angularSpringConstant=20000
    bAngularSpring=true
    bOverrideAngularDamping=true
    bUseGravityOverride=true
    bUsePlanarLimit=false
    bUseSphericalLimits=true
    gravityScale=0.35
    simulationSpace=RootRelative
    alphaBoolBlend {blendOption=Linear}
    boundBone {boneName=dyn_collar_bk_01}
    componentLinearAccScale=<0.7 0.7 0.7>
    componentLinearVelScale=<0.7 0.7 0.7>
    gravityOverride {z=-980}
    physicsBodyDefinitions [{
      collisionType=CustomSphere
      sphereCollisionRadius=2
      boundBone {boneName=dyn_collar_bk_01}
      boxExtents=<20 30 30>
      constraintSetup {angularLimitsMax=<0 5 0>, angularLimitsMin=<0 -3 0>, angularTarget=<1 0 0>}
      localJointOffset=<4 0 0>
     }]
    sphericalLimits [{limitRadius=7.5, drivingBone {boneName=head}, sphereLocalOffset=<1 0 0>}]
   }
   n31 = Anim Dynamics - Bone: dyn_collar_01_R!(ComponentPose=@n30, Alpha=1, GravityOverride:Vector="0,0,-980") -> (Pose) {
    angularDampingOverride=1
    angularSpringConstant=20000
    bAngularSpring=true
    bOverrideAngularDamping=true
    bUseGravityOverride=true
    bUsePlanarLimit=false
    bUseSphericalLimits=true
    gravityScale=0.35
    simulationSpace=RootRelative
    alphaBoolBlend {blendOption=Linear}
    boundBone {boneName=dyn_collar_01_R}
    componentLinearAccScale=<0.7 0.7 0.7>
    componentLinearVelScale=<0.7 0.7 0.7>
    gravityOverride {z=-980}
    physicsBodyDefinitions [{
      collisionType=CustomSphere
      sphereCollisionRadius=2
      boundBone {boneName=dyn_collar_01_R}
      boxExtents=<20 30 30>
      constraintSetup {angularLimitsMax=<0 5 0>, angularLimitsMin=<0 -3 0>, angularTarget=<1 0 0>}
      localJointOffset=<4 0 0>
     }]
    sphericalLimits [
     {limitRadius=7.5, drivingBone {boneName=clavicle_r}, sphereLocalOffset=<-8.9578 -0.8451 -4.143>}
     {limitRadius=6, drivingBone {boneName=head}, sphereLocalOffset=<0 0 0>}
    ]
   }
   n32 = "Transform (Modify) Bone - Bone: dyn_strap_MSTR"!(ComponentPose=@n31, Alpha=Get Strap bending when crouching() -> (Strap bending when crouching:float), Rotation="60,0,0") -> (Pose) {
    rotationMode=BMM_Additive
    rotationSpace=BCS_BoneSpace
    alphaBoolBlend {blendOption=Linear}
    alphaScaleBiasClamp {bInterpResult=true, interpSpeedDecreasing=5, interpSpeedIncreasing=5}
    boneToModify {boneName=dyn_strap_MSTR}
    rotation {pitch=60}
   }
   n33 = Get Gravity Override()
   n34 = Get crouching strap() -> (crouching strap:double)
   n35 = Anim Dynamics - Bone: dyn_strap_01!(ComponentPose=@n32, Alpha=@n34, GravityOverride:Vector=@n33) -> (Pose) {
    angularBiasOverride=1
    angularDampingOverride=1
    angularSpringConstant=1000
    bAngularSpring=true
    bOverrideAngularBias=true
    bOverrideAngularDamping=true
    bUseGravityOverride=true
    simulationSpace=RootRelative
    alphaBoolBlend {blendOption=Linear}
    boundBone {boneName=dyn_strap_01}
    componentAppliedLinearAccClamp=<3000 3000 3000>
    componentLinearAccScale=<0.1 0.1 0.1>
    componentLinearVelScale=<0.1 0.1 0.1>
    physicsBodyDefinitions [{
      collisionType=CustomSphere
      sphereCollisionRadius=2
      boundBone {boneName=dyn_strap_01}
      boxExtents=<10 10 10>
      constraintSetup {angularLimitsMax=<0 25 0>, angularLimitsMin=<0 -25 0>, angularTarget=<0 0 0>}
      localJointOffset=<5 -0.4641 0.5502>
     }]
    planarLimits [{drivingBone {boneName=dyn_coat_fr_mid_01_R}, planeTransform {location=<-0.0002 5.6243 -1.4234>, rotation=<rot 0.0013 0.0001 5.0024>}}]
    sphericalLimits [
     {limitRadius=7, drivingBone {boneName=dyn_coat_fr_mid_01_R}, sphereLocalOffset=<9.1685 0.7653 -4.5416>}
     {limitRadius=5, drivingBone {boneName=dyn_coat_fr_01_R}, sphereLocalOffset=<10.6205 -0.1833 -3.8411>}
    ]
   }
   n36 = Anim Dynamics - Bone: dyn_strap_02!(ComponentPose=@n35, Alpha=@n34, GravityOverride:Vector=@n33) -> (Pose) {
    angularBiasOverride=1
    angularDampingOverride=1
    angularSpringConstant=1000
    bAngularSpring=true
    bOverrideAngularBias=true
    bOverrideAngularDamping=true
    bUseGravityOverride=true
    simulationSpace=RootRelative
    alphaBoolBlend {blendOption=Linear}
    boundBone {boneName=dyn_strap_02}
    componentAppliedLinearAccClamp=<3000 3000 3000>
    componentLinearAccScale=<0.1 0.1 0.1>
    componentLinearVelScale=<0.1 0.1 0.1>
    physicsBodyDefinitions [{
      collisionType=CustomSphere
      sphereCollisionRadius=2
      boundBone {boneName=dyn_strap_02}
      boxExtents=<10 10 10>
      constraintSetup {angularLimitsMax=<0 35 0>, angularLimitsMin=<0 -35 0>, angularTarget=<0 0 0>}
      localJointOffset=<5 0 0>
     }]
    planarLimits [{drivingBone {boneName=dyn_coat_fr_mid_01_R}, planeTransform {location=<10.6558 -0.1416 0.1442>, rotation=<rot -9.9583 -0.8981 5.0808>}}]
    sphericalLimits [{limitRadius=8, drivingBone {boneName=dyn_coat_fr_mid_01_R}, sphereLocalOffset=<13.5393 -3.3142 -4.2153>}]
   }
   Output Pose(Result=Blend Poses by bool(BlendPose_0=Component To Local(ComponentPose=@n36), BlendPose_1=Component To Local(ComponentPose=@n1), bActiveValue=Get bRigidBodyEnabled()) {blendType=Linear}) {layerGroup=None}
   n37 = Control Rig(Source=Copy Pose From Mesh() {bUseAttachedParent=true}) {
    alphaBoolBlend {blendOption=Linear}
    controlRigAssetReference {blueprintRigClass=&/M_MED_NovelGuard_ControlRig.M_MED_NovelGuard_ControlRig_C}
   }
   n38 = "Constraint - badge_CST (PPPPPPPPPPP)"!(
    ComponentPose=Local To Component(LocalPose=@n37)
    Alpha=1
    ConstraintWeights_0=0.0001
    ConstraintWeights_1:float=0.0011
    ConstraintWeights_2:float=0.1748
    ConstraintWeights_3:float=0.5849
    ConstraintWeights_4:float=0.2314
    ConstraintWeights_5:float=0
    ConstraintWeights_6:float=0.0077
    ConstraintWeights_7:float=0
    ConstraintWeights_8:float=0
    ConstraintWeights_9:float=0
    ConstraintWeights_10:float=0
   ) -> (Pose) {
    alphaBoolBlend {blendOption=Linear}
    boneToModify {boneName=badge_CST}
    constraintSetup [
     {targetBone {boneName=pelvis}}
     {targetBone {boneName=spine_02}}
     {targetBone {boneName=spine_04}}
     {targetBone {boneName=spine_05}}
     {targetBone {boneName=clavicle_l}}
     {targetBone {boneName=upperarm_l}}
     {targetBone {boneName=upperarm_twist_01_l}}
     {targetBone {boneName=clavicle_r}}
     {targetBone {boneName=upperarm_twist_01_r}}
     {targetBone {boneName=dyn_collar_fr_02_R}}
     {targetBone {boneName=dyn_collar_01_R}}
    ]
    constraintWeights [1,1,1,1,1,1,1,1,1,1,1]
   }
   n39 = "Transform (Modify) Bone - Bone: pad_base"!(ComponentPose=@n38, Alpha=1, Rotation=Get pad rotation() -> (pad rotation:Rotator)) -> (Pose) {
    rotationMode=BMM_Additive
    rotationSpace=BCS_BoneSpace
    translationSpace=BCS_BoneSpace
    alphaBoolBlend {blendOption=Linear}
    boneToModify {boneName=pad_base}
   }
   n40 = "Transform (Modify) Bone - Bone: pad_01"(ComponentPose=@n39, Alpha=Get upperarm r rotation z() -> (upperarm r rotation z:double), Translation="-0.114,-0.8,0", Rotation="-13,0,10") {
    rotationMode=BMM_Additive
    rotationSpace=BCS_BoneSpace
    translationMode=BMM_Additive
    translationSpace=BCS_BoneSpace
    alphaBoolBlend {blendOption=Linear}
    alphaScaleBiasClamp {bInterpResult=true, interpSpeedDecreasing=5, interpSpeedIncreasing=5}
    boneToModify {boneName=pad_01}
    rotation {pitch=-13, roll=10}
    translation=<-0.114 -0.8>
   }
   n41 = "Get is skydiving "() -> ("is skydiving ":double)
   n42 = "Transform (Modify) Bone - Bone: pad_01"(ComponentPose=@n40, Alpha=@n41, Translation="-6,1,-2", Rotation="0,0,0") {
    rotationMode=BMM_Additive
    rotationSpace=BCS_BoneSpace
    translationMode=BMM_Additive
    translationSpace=BCS_BoneSpace
    alphaBoolBlend {blendOption=Linear}
    alphaScaleBiasClamp {interpSpeedDecreasing=1, interpSpeedIncreasing=5}
    boneToModify {boneName=pad_01}
    translation=<-6 1 -2>
   }
   n43 = "Transform (Modify) Bone - Bone: pad_base"!(ComponentPose=@n42, Alpha=@n41, Translation:Vector="0,0,0", Rotation="0,0,-20") -> (Pose) {
    rotationMode=BMM_Additive
    rotationSpace=BCS_BoneSpace
    translationMode=BMM_Additive
    translationSpace=BCS_BoneSpace
    alphaBoolBlend {blendOption=Linear}
    alphaScaleBiasClamp {interpSpeedDecreasing=1, interpSpeedIncreasing=5}
    boneToModify {boneName=pad_base}
    rotation {roll=-20}
   }
   n44 = "Transform (Modify) Bone - Bone: hand_armor"!(ComponentPose=@n43, Alpha=Get lower arm bending() -> (lower arm bending:float), Translation="-5,0,0") -> (Pose) {
    rotationSpace=BCS_BoneSpace
    translationMode=BMM_Additive
    translationSpace=BCS_BoneSpace
    alphaBoolBlend {blendOption=Linear}
    boneToModify {boneName=hand_armor}
    rotation {pitch=150}
    translation {x=-5}
   }
   Save cached pose 'main cache'(Pose=Component To Local(ComponentPose=@n44)) {cacheName=main cache}
   n45 = Get sliding() -> (sliding:double)
   n46 = "Transform (Modify) Bone - Bone: dyn_coat_MSTR_R"!(ComponentPose=Local To Component(LocalPose=Use cached pose 'main cache'()), Alpha=@n45, Rotation="50,0,0") -> (Pose) {
    rotationMode=BMM_Additive
    rotationSpace=BCS_BoneSpace
    alphaBoolBlend {blendOption=Linear}
    alphaScaleBiasClamp {bInterpResult=true}
    boneToModify {boneName=dyn_coat_MSTR_R}
    rotation {pitch=50}
   }
   n47 = "Transform (Modify) Bone - Bone: dyn_coat_bk_mid_MSTR_R"(ComponentPose=@n46, Alpha=@n45, Rotation="50,-50,-30") {
    rotationMode=BMM_Additive
    rotationSpace=BCS_BoneSpace
    alphaBoolBlend {blendOption=Linear}
    alphaScaleBiasClamp {bInterpResult=true}
    boneToModify {boneName=dyn_coat_bk_mid_MSTR_R}
    rotation=<rot 50 -50 -30>
   }
   n48 = "Transform (Modify) Bone - Bone: dyn_coat_bk_MSTR_R"(ComponentPose=@n47, Alpha=@n45, Rotation="50,-50,-30") {
    rotationMode=BMM_Additive
    rotationSpace=BCS_BoneSpace
    alphaBoolBlend {blendOption=Linear}
    alphaScaleBiasClamp {bInterpResult=true}
    boneToModify {boneName=dyn_coat_bk_MSTR_R}
    rotation=<rot 50 -50 -30>
   }
   n49 = "Transform (Modify) Bone - Bone: dyn_coat_MSTR_L"!(ComponentPose=@n48, Alpha=@n45, Rotation="50,0,0") -> (Pose) {
    rotationMode=BMM_Additive
    rotationSpace=BCS_BoneSpace
    alphaBoolBlend {blendOption=Linear}
    alphaScaleBiasClamp {bInterpResult=true}
    boneToModify {boneName=dyn_coat_MSTR_L}
    rotation {pitch=50}
   }
   n50 = "Transform (Modify) Bone - Bone: dyn_coat_bk_mid_MSTR_L"(ComponentPose=@n49, Alpha=@n45, Rotation="50,0,0") {
    rotationMode=BMM_Additive
    rotationSpace=BCS_BoneSpace
    alphaBoolBlend {blendOption=Linear}
    alphaScaleBiasClamp {bInterpResult=true}
    boneToModify {boneName=dyn_coat_bk_mid_MSTR_L}
    rotation {pitch=50}
   }
   n51 = "Transform (Modify) Bone - Bone: dyn_coat_bk_MSTR_L"(ComponentPose=@n50, Alpha=@n45, Rotation="50,0,0") {
    rotationMode=BMM_Additive
    rotationSpace=BCS_BoneSpace
    alphaBoolBlend {blendOption=Linear}
    alphaScaleBiasClamp {bInterpResult=true}
    boneToModify {boneName=dyn_coat_bk_MSTR_L}
    rotation {pitch=50}
   }
   n52 = "Transform (Modify) Bone - Bone: dyn_sword_MSTR"!(ComponentPose=@n51, Alpha=@n45, Rotation="10,0,0") -> (Pose) {
    rotationMode=BMM_Additive
    rotationSpace=BCS_BoneSpace
    alphaBoolBlend {blendOption=Linear}
    alphaScaleBiasClamp {bInterpResult=true}
    boneToModify {boneName=dyn_sword_MSTR}
    rotation {pitch=10}
   }
   n53 = "Get is skydiving "()
   n54 = "Transform (Modify) Bone - Bone: dyn_coat_bk_MSTR_L"(ComponentPose=@n52, Alpha=@n53, Rotation="0,10,0") {
    rotationMode=BMM_Additive
    rotationSpace=BCS_BoneSpace
    alphaBoolBlend {blendOption=Linear}
    boneToModify {boneName=dyn_coat_bk_MSTR_L}
    rotation {yaw=10}
   }
   n55 = "Transform (Modify) Bone - Bone: dyn_coat_bk_MSTR_R"(ComponentPose=@n54, Alpha=@n53, Rotation="0,-10,0") {
    rotationMode=BMM_Additive
    rotationSpace=BCS_BoneSpace
    alphaBoolBlend {blendOption=Linear}
    boneToModify {boneName=dyn_coat_bk_MSTR_R}
    rotation {yaw=-10}
   }
   n56 = "Transform (Modify) Bone - Bone: dyn_coat_bk_mid_MSTR_L"(ComponentPose=@n55, Alpha=@n53, Rotation="0,10,0") {
    rotationMode=BMM_Additive
    rotationSpace=BCS_BoneSpace
    alphaBoolBlend {blendOption=Linear}
    boneToModify {boneName=dyn_coat_bk_mid_MSTR_L}
    rotation {yaw=10}
   }
   n57 = "Transform (Modify) Bone - Bone: dyn_coat_bk_mid_MSTR_R"(ComponentPose=@n56, Alpha=@n53, Rotation="0,-10,0") {
    rotationMode=BMM_Additive
    rotationSpace=BCS_BoneSpace
    alphaBoolBlend {blendOption=Linear}
    boneToModify {boneName=dyn_coat_bk_mid_MSTR_R}
    rotation {yaw=-10}
   }
   n58 = "Transform (Modify) Bone - Bone: belt"!(ComponentPose=@n57, Alpha=Get Front Bending() -> (Front Bending:double), Translation="4.5,0,0") -> (Pose) {
    rotationSpace=BCS_BoneSpace
    translationMode=BMM_Additive
    translationSpace=BCS_BoneSpace
    alphaBoolBlend {blendOption=Linear}
    alphaScaleBiasClamp {bInterpResult=true, interpSpeedIncreasing=20}
    boneToModify {boneName=belt}
    translation {x=4.5}
   }
   n59 = "Transform (Modify) Bone - Bone: belt"(ComponentPose=@n58, Alpha=Get thigh bending crouching not moving() -> (thigh bending crouching not moving:double), Translation="5,0,0") {
    rotationSpace=BCS_BoneSpace
    translationMode=BMM_Additive
    translationSpace=BCS_BoneSpace
    alphaBoolBlend {blendOption=Linear}
    alphaScaleBiasClamp {bInterpResult=true, interpSpeedDecreasing=15, interpSpeedIncreasing=20}
    boneToModify {boneName=belt}
    translation {x=5}
   }
   n60 = "Transform (Modify) Bone - Bone: belt"(ComponentPose=@n59, Alpha="Get tactical sprinting "() -> ("tactical sprinting ":double), Translation="5,0,0") {
    rotationSpace=BCS_BoneSpace
    translationMode=BMM_Additive
    translationSpace=BCS_BoneSpace
    alphaBoolBlend {blendOption=Linear}
    alphaScaleBiasClamp {bInterpResult=true, interpSpeedDecreasing=15, interpSpeedIncreasing=5}
    boneToModify {boneName=belt}
    translation {x=5}
   }
   n61 = "Transform (Modify) Bone - Bone: belt"(ComponentPose=@n60, Alpha=Get sliding(), Translation="3,0,0") {
    rotationSpace=BCS_BoneSpace
    translationMode=BMM_Additive
    translationSpace=BCS_BoneSpace
    alphaBoolBlend {blendOption=Linear}
    alphaScaleBiasClamp {bInterpResult=true, interpSpeedDecreasing=15, interpSpeedIncreasing=20}
    boneToModify {boneName=belt}
    translation {x=3}
   }
   n62 = Get Thigh l bending() -> (Thigh l bending:float)
   n63 = "Transform (Modify) Bone - Bone: dyn_coat_fr_MSTR_L"(ComponentPose=@n61, Alpha=@n62, Rotation="32.731,-105.095,-91.350") {
    rotationMode=BMM_Additive
    rotationSpace=BCS_BoneSpace
    alphaBoolBlend {blendOption=Linear}
    boneToModify {boneName=dyn_coat_fr_MSTR_L}
    rotation=<rot 32.731 -105.095 -91.3501>
   }
   n64 = "Transform (Modify) Bone - Bone: dyn_coat_fr_mid_MSTR_L"(ComponentPose=@n63, Alpha=@n62, Rotation="17.980,8.851,-68.157") {
    rotationMode=BMM_Additive
    rotationSpace=BCS_BoneSpace
    alphaBoolBlend {blendOption=Linear}
    boneToModify {boneName=dyn_coat_fr_mid_MSTR_L}
    rotation=<rot 17.98 8.851 -68.157>
   }
   n65 = Get Thigh r bending() -> (Thigh r bending:float)
   n66 = "Transform (Modify) Bone - Bone: dyn_coat_fr_MSTR_R"(ComponentPose=@n64, Alpha=@n65, Rotation="77.602,108.020,110.964") {
    rotationMode=BMM_Additive
    rotationSpace=BCS_BoneSpace
    alphaBoolBlend {blendOption=Linear}
    boneToModify {boneName=dyn_coat_fr_MSTR_R}
    rotation=<rot 77.602 108.02 110.964>
   }
   n67 = "Transform (Modify) Bone - Bone: dyn_coat_fr_mid_MSTR_R"(ComponentPose=@n66, Alpha=@n65, Rotation="58.233,17.753,11.389") {
    rotationMode=BMM_Additive
    rotationSpace=BCS_BoneSpace
    alphaBoolBlend {blendOption=Linear}
    boneToModify {boneName=dyn_coat_fr_mid_MSTR_R}
    rotation=<rot 58.2332 17.7528 11.3891>
   }
   Save cached pose 'presim cache'(Pose=Component To Local(ComponentPose=@n67)) {cacheName=presim cache}
  ]}
 EventGraph {nodes [
   Event BlueprintUpdateAnimation()
   > n1 = Sequence!(execute) -> (then_0, then_1, then_2:Exec, then_3:Exec)
   > n2 = Gravity Override State Function DEPRECATED(CharacterPartAnimInstance=Self-Reference(), GameTimeInSeconds=GetGameTimeInSeconds(), GravityOverrideParams=Get Gravity Override Params())
   > Set Gravity Override(Gravity Override=@n2)
   n3 = SelectVector(A="0.1,0.1,0.1", B="0.35,0.35,0.35", bPickA=OR Boolean(A=Get bIsSkyDiving(), B=Get bIsTacticalSprinting()))
   > Set ComponentLinearAccScale(ComponentLinearAccScale:Vector=@n3) -> (Output_Get:Vector)
   > Set ComponentLinearVelScale(ComponentLinearVelScale:Vector=@n3) -> (Output_Get:Vector)
   n4 = Get bIsCrouching()
   n5 = Get bIsMoving2D()
   > Set Crouching moving(Crouching moving:double=SelectFloat(A=1, B=0, bPickA=AND Boolean(A=@n4, B=@n5))) -> (Output_Get:double)
   > Set crouching strap(crouching strap:double=SelectFloat(A=0.58, B=1, bPickA=Get bIsCrouching())) -> (Output_Get:double)
   > Set sliding(sliding:double=SelectFloat(A=1, B=0, bPickA=Get bIsSliding())) -> (Output_Get:double)
   > "Set is skydiving "("is skydiving ":double=SelectFloat(A=1, B=0, bPickA=Get bIsSkyDiving())) -> (Output_Get:double)
   > "Set tactical sprinting "("tactical sprinting ":double=SelectFloat(A=1, B=0, bPickA=Get bIsTacticalSprinting())) -> (Output_Get:double)
   n6 = Get bIsSkyDiving()
   n7 = Get bIsJumping()
   n8 = Get bIsCrouching()
   n9 = Get bIsFalling()
   n10 = AND Boolean(A=Get bIsSprinting(), B=NOT Boolean(A=OR Boolean!(A=@n8, B=@n6, C:bool=@n7, D:bool=@n9) -> (ReturnValue)))
   n11 = Get bIsJumping()
   n12 = Get bIsFalling()
   > Set moving(moving:double=SelectFloat(A=0.5, B=SelectFloat(A=0.7, B=0, bPickA=@n10), bPickA=OR Boolean!(A=@n11, B=@n12) -> (ReturnValue))) -> (Output_Get:double)
   n13 = Get SkelMesh()
   Set lower arm bending(
    execute=@n1.then_3
    lower arm bending:float=GetAlphaFromReferenceTransformDelta(SkeletalMeshComponent=@n13, BoneName=lowerarm_r, BaseName=upperarm_r, InputRange=(X=50,Y=90), TwistAxis="1,0,0")
   ) -> (Output_Get:float)
   > Set head pitch(head pitch:float=GetAlphaFromReferenceTransformDelta(SkeletalMeshComponent=@n13, BoneName=head, BaseName=spine_05, InputRange=(X=-30,Y=-100), TwistAxis="1,0,0")) -> (Output_Get:float)
   n14 = BreakTransform!(InTransform=GetSocketTransform(self=@n13, InSocketName=upperarm_twist_02_r, TransformSpace=RTS_ParentBoneSpace)) -> (Location, Rotation_Roll:float, Rotation_Pitch:float, Rotation_Yaw:float, Scale)
   > Set!(execute, pad rotation_Roll:float=@n14.Rotation_Roll, pad rotation_Pitch:float=0, pad rotation_Yaw:float=@n14.Rotation_Yaw) -> (then, Output_Get:Rotator)
   n15 = BreakTransform(InTransform=GetSocketTransform(self=@n13, InSocketName=upperarm_r, TransformSpace=RTS_ParentBoneSpace))
   > Set upperarm r rotation z(upperarm r rotation z:double=SelectFloat(A=MapRangeClamped(Value=@n15.Rotation_Yaw, InRangeA=70, InRangeB=0), B=0, bPickA=Get bIsCrouching())) -> (Output_Get:double)
   n16 = Get bIsSkyDiving()
   > Set upperarm r rotation y(upperarm r rotation y:double=SelectFloat(A=MapRangeClamped(Value=@n15.Rotation_Pitch, InRangeA=-20, InRangeB=0), B=0, bPickA=NOT Boolean(A=@n16))) -> (Output_Get:double)
   n17 = Get SkelMesh()
   n18 = float + float(
    A=GetAlphaFromReferenceTransformDelta(SkeletalMeshComponent=@n17, BoneName=thigh_l, BaseName=pelvis, InputRange=(X=-50,Y=-90), TwistAxis="1,0,0")
    B=GetAlphaFromReferenceTransformDelta(SkeletalMeshComponent=@n17, BoneName=thigh_r, BaseName=pelvis, InputRange=(X=-50,Y=-90), TwistAxis="1,0,0")
   )
   n19 = "float > float"(A=@n18, B=1)
   n20 = NOT Boolean(A=OR Boolean(A=Get bIsCrouching(), B=Get bIsSprinting()))
   n21 = Get bIsFalling()
   n22 = Get bIsJumping()
   Set Front Bending(
    execute=@n1.then_1
    Front Bending:double=SelectFloat(A=SelectFloat(A=1, B=@n18, bPickA=@n19), B=0, bPickA=OR Boolean!(A=@n20, B=@n22, C:bool=@n21) -> (ReturnValue))
   ) -> (Output_Get:double)
   n23 = float + float(
    A=GetAlphaFromReferenceTransformDelta(SkeletalMeshComponent=@n17, BoneName=thigh_l, BaseName=pelvis, InputRange=(X=-80,Y=-110), TwistAxis="1,0,0")
    B=GetAlphaFromReferenceTransformDelta(SkeletalMeshComponent=@n17, BoneName=thigh_r, BaseName=pelvis, InputRange=(X=-80,Y=-110), TwistAxis="1,0,0")
   )
   n24 = "float > float"(A=@n23, B=1)
   n25 = NOT Boolean(A=Get bIsMoving2D())
   n26 = Get bIsCrouching()
   > Set thigh bending crouching not moving(thigh bending crouching not moving:double=SelectFloat(A=SelectFloat(A=1, B=@n23, bPickA=@n24), B=0, bPickA=AND Boolean(A=@n26, B=@n25))) -> (Output_Get:double)
   n27 = Get SkelMesh()
   n28 = GetAlphaFromReferenceTransformDelta(SkeletalMeshComponent=@n27, BoneName=thigh_l, BaseName=pelvis, InputRange=(X=0,Y=-100), TwistAxis="1,0,0")
   n29 = NOT Boolean(A=OR Boolean(A=Get bIsCrouching(), B=Get bIsJumping(), C=Get bIsFalling()))
   n30 = OR Boolean!(A=Get bIsSprinting(), B=Get bIsTacticalSprinting()) -> (ReturnValue)
   Set Thigh l bending(execute=@n1.then_2, Thigh l bending:float=SelectFloat(A=float / float(A=@n28, B=5), B=@n28, bPickA=AND Boolean(A=@n30, B=@n29))) -> (Output_Get:float)
   n31 = GetAlphaFromReferenceTransformDelta(SkeletalMeshComponent=@n27, BoneName=thigh_r, BaseName=pelvis, InputRange=(X=0,Y=-100), TwistAxis="1,0,0")
   n32 = NOT Boolean(A=OR Boolean!(A=Get bIsCrouching(), B=Get bIsJumping(), C:bool=Get bIsFalling()) -> (ReturnValue))
   n33 = OR Boolean!(A=Get bIsSprinting(), B=Get bIsTacticalSprinting()) -> (ReturnValue)
   > Set Thigh r bending(Thigh r bending:float=SelectFloat(A=float / float(A=@n31, B=5), B=@n31, bPickA=AND Boolean(A=@n33, B=@n32))) -> (Output_Get:float)
   n34 = Get bIsCrouching()
   n35 = Get bIsMoving2D()
   > Set Strap bending when crouching(Strap bending when crouching:float=SelectFloat(A=@n31, B=0, bPickA=AND Boolean(A=@n34, B=@n35))) -> (Output_Get:float)
   > Set Thigh bending when sprinting(Thigh bending when sprinting:float=SelectFloat(A=@n31, B=0, bPickA=Get bIsTacticalSprinting())) -> (Output_Get:float)
   n36 = NOT Boolean(A=Get bIsMoving2D())
   n37 = Get bIsCrouching()
   > Set crouching sword motion(crouching sword motion:double=SelectFloat(
     A=GetAlphaFromReferenceTransformDelta(SkeletalMeshComponent=@n27, BoneName=thigh_r, BaseName=pelvis, InputRange=(X=20,Y=-100), TwistAxis="1,0,0")
     B=0
     bPickA=AND Boolean(A=@n37, B=@n36)
    )) -> (Output_Get:double)
   n38 = Event OnCharacterCustomizationLoaded()
   n39 = Get Dynamic Material Instance(execute=[@n38.then, @n43.then], SkeletalMeshComponent=GetPartSkeletalMeshComponent())
   n40 = PickaxeItemDefinitionHasAnyMetaTag(Tags=Make Array("[0]"=(TagName="Cosmetics.MetaTags.NovelGuard")), bRequiresEquipped=true)
   > n41 = Branch(Condition=@n40)
   > n42 = Branch(Condition=Get bIsSkyDiving())
   SetScalarParameterValue(execute=[@n44.then, @n45.else, @n42.then, @n41.else], self=@n39, ParameterName=Hide Element 01, Value=0)
   SetScalarParameterValue(execute=[@n45.then, @n42.else], self=@n39, ParameterName=Hide Element 01, Value=1)
   n43 = Event OnWeaponChanged()
   n44 = Event OnBeginSkydiving()
   Event OnEndSkydiving()
   > n45 = Branch(Condition=@n40)
   Event BlueprintInitializeAnimation()
   > Set SkelMesh(SkelMesh=GetOwnerSkeletalMeshComponent())
   Get bIsSprinting()
  ]}
]
variables [
 ComponentLinearAccScale {type=Vector}
 ComponentLinearVelScale {type=Vector}
 Crouching moving {type=number}
 crouching strap {type=number}
 crouching sword motion {type=number}
 Front Bending {type=number}
 Gravity Override {type=Vector}
 Gravity Override Params {type=GravityOverrideParamsStruct}
 head pitch {type=number}
 "is skydiving " {type=number}
 lower arm bending {type=number}
 moving {type=number}
 pad rotation {type=Rotator}
 SkelMesh {type=SkeletalMeshComponent}
 sliding {type=number}
 Strap bending when crouching {type=number}
 "tactical sprinting " {type=number}
 thigh bending crouching not moving {type=number}
 Thigh bending when sprinting {type=number}
 Thigh l bending {type=number}
 Thigh r bending {type=number}
 upperarm r rotation y {type=number}
 upperarm r rotation z {type=number}
]
