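# Physics asset of collision shapes for the cape on the medium male NovelGuard player body, going by its name.
# Capsules cover pelvis, spine, neck and clavicles, with a sphere on the head; arm and leg bodies carry no shapes.
# Extra capsules follow the dyn_coat bones (front right, both back sides) and dyn_sword_02, so those dynamic parts also act as colliders.
# Constraints link only the body skeleton, all with free swing and twist.
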PhysicsAsset /Game/Characters/Player/Male/Medium/Bodies/M_MED_NovelGuard/Meshes/M_MED_NovelGuard_CapeColliders
bodies {
 calf_l []
 calf_r []
 clavicle_l [{shapeType=Capsule, length=9.0192, radius=9.8871, center=<10.1596 0.3235 -0.2302>, rotation=<rot 69.3297 1.4476 176.9462>}]
 clavicle_r [{shapeType=Capsule, length=7.5192, radius=11.3871, center=<-10.8447 -0.3467 -1.2697>, rotation=<rot 64.3355 2.0615 -2.4889>}]
 dyn_coat_01_R [{shapeType=Capsule, length=6, radius=5.5, shapeName=dyn_coat_01_R_capsule, center=<0 0 -4.4519>, rotation=<rot 90 0 0>}]
 dyn_coat_02_R [{shapeType=Capsule, length=6, radius=6, shapeName=dyn_coat_02_R_capsule, center=<0 0 -4.4519>, rotation=<rot 90 0 0>}]
 dyn_coat_03_R [{shapeType=Capsule, length=10, radius=6, shapeName=dyn_coat_03_R_capsule, center=<0.9768 0 -4.4519>, rotation=<rot 90 0 0>}]
 dyn_coat_04_R [{shapeType=Capsule, length=20, radius=6, shapeName=dyn_coat_04_R_capsule, center=<7.4767 0 -3.2035>, rotation=<rot 87.1645 179.9987 179.9992>}]
 dyn_coat_bk_01_L [{shapeType=Capsule, length=6, radius=5.5, shapeName=dyn_coat_bk_01_L_capsule, center=<0 0 -4.4519>, rotation=<rot 90 0 0>}]
 dyn_coat_bk_01_R [{shapeType=Capsule, length=6, radius=5.5, shapeName=dyn_coat_bk_01_R_capsule, center=<0 0 -4.4519>, rotation=<rot 90 0 0>}]
 dyn_coat_bk_02_L [{shapeType=Capsule, length=6, radius=6, shapeName=dyn_coat_bk_02_L_capsule, center=<0 0 -4.4519>, rotation=<rot 90 0 0>}]
 dyn_coat_bk_02_R [{shapeType=Capsule, length=6, radius=8, shapeName=dyn_coat_bk_02_R_capsule, center=<0 0 -5.1182>, rotation=<rot 90 0 0>}]
 dyn_coat_bk_03_L [{shapeType=Capsule, length=23, radius=9, shapeName=dyn_coat_bk_03_L_capsule, center=<5.5165 -1.6467 -7.2276>, rotation=<rot 90 0 0>}]
 dyn_coat_bk_03_R [{shapeType=Capsule, length=13, radius=8, shapeName=dyn_coat_bk_03_R_capsule, center=<0.9768 2.1004 -4.4519>, rotation=<rot 90 0 0>}]
 dyn_coat_bk_04_R [{
   shapeType=Capsule
   length=23
   radius=10
   shapeName=dyn_coat_bk_04_R_capsule
   center=<9.0847 1.9108 -6.9095>
   rotation=<rot 87.1649 179.9993 174.9994>
  }]
 dyn_coat_bk_mid_01_L [{shapeType=Capsule, length=6, radius=5.5, shapeName=dyn_coat_bk_mid_01_L_capsule, center=<0 0 -4.4519>, rotation=<rot 90 0 0>}]
 dyn_coat_bk_mid_01_R [{shapeType=Capsule, length=6, radius=5.5, shapeName=dyn_coat_bk_mid_01_R_capsule, center=<0 0 -4.4519>, rotation=<rot 90 0 0>}]
 dyn_coat_bk_mid_02_L [{shapeType=Capsule, length=6, radius=6, shapeName=dyn_coat_bk_mid_02_L_capsule, center=<0 0 -4.4519>, rotation=<rot 90 0 0>}]
 dyn_coat_bk_mid_02_R [{shapeType=Capsule, length=6, radius=6, shapeName=dyn_coat_bk_mid_02_R_capsule, center=<0 0 -4.4519>, rotation=<rot 90 0 0>}]
 dyn_coat_bk_mid_03_L [{shapeType=Capsule, length=23, radius=6, shapeName=dyn_coat_bk_mid_03_L_capsule, center=<5.5165 0 -4.4519>, rotation=<rot 90 0 0>}]
 dyn_coat_bk_mid_03_R [{shapeType=Capsule, length=10, radius=6, shapeName=dyn_coat_bk_mid_03_R_capsule, center=<0.9768 0 -4.4519>, rotation=<rot 90 0 0>}]
 dyn_coat_bk_mid_04_R [{
   shapeType=Capsule
   length=20
   radius=6
   shapeName=dyn_coat_bk_mid_04_R_capsule
   center=<8.3869 0 -3.2035>
   rotation=<rot 87.1647 179.9991 179.9994>
  }]
 dyn_sword_02 [{shapeType=Capsule, length=60, radius=7.5, shapeName=dyn_sword_02_capsule, center=<42.3513 -1.1993 0>, rotation=<rot 90 0 0>}]
 foot_l []
 foot_r []
 head [{shapeType=Sphere, radius=17.3125, center=<9.9016 -0.7947 0>}]
 lowerarm_l []
 lowerarm_r []
 neck_02 [{shapeType=Capsule, length=12.9375, radius=7.8175, center=<-0.3797 -1.4928 0>, rotation=<rot 90 0 30.7218>}]
 pelvis [{shapeType=Capsule, length=8.6456, radius=14.0388, center=<-2.4887 0.1881 0>}]
 spine_01 [{shapeType=Capsule, length=7.9438, radius=13.5586, center=<-1.7965 1.1125 0>}]
 spine_02 [{shapeType=Capsule, length=7.3264, radius=12.1322, center=<-1.711 1.991 0.0001>}]
 spine_03 [{shapeType=Capsule, length=8.2107, radius=11.7031, center=<0.1445 1.8227 0.0001>}]
 spine_04 [{shapeType=Capsule, length=9.4404, radius=12.6325, center=<2.766 1.9704 0.0002>}]
 spine_05 [
  {shapeType=Capsule, length=8.0226, radius=15.695, center=<-2.2795 2.5642 0>}
  {shapeType=Capsule, length=13.9518, radius=13.9425, center=<-0.2916 0.8839 0>}
 ]
 thigh_l []
 thigh_r []
 upperarm_l []
 upperarm_r []
}
constraints [
 {
  bone1Name=spine_01
  bone2Name=pelvis
  swing1LimitDegrees=45
  swing1Motion=Free
  swing2LimitDegrees=45
  swing2Motion=Free
  twistLimitDegrees=45
  twistMotion=Free
 }
 {
  bone1Name=spine_02
  bone2Name=spine_01
  swing1LimitDegrees=45
  swing1Motion=Free
  swing2LimitDegrees=45
  swing2Motion=Free
  twistLimitDegrees=45
  twistMotion=Free
 }
 {
  bone1Name=spine_03
  bone2Name=spine_02
  swing1LimitDegrees=45
  swing1Motion=Free
  swing2LimitDegrees=45
  swing2Motion=Free
  twistLimitDegrees=45
  twistMotion=Free
 }
 {
  bone1Name=spine_04
  bone2Name=spine_03
  swing1LimitDegrees=45
  swing1Motion=Free
  swing2LimitDegrees=45
  swing2Motion=Free
  twistLimitDegrees=45
  twistMotion=Free
 }
 {
  bone1Name=spine_05
  bone2Name=spine_04
  swing1LimitDegrees=45
  swing1Motion=Free
  swing2LimitDegrees=45
  swing2Motion=Free
  twistLimitDegrees=45
  twistMotion=Free
 }
 {
  bone1Name=neck_02
  bone2Name=spine_05
  swing1LimitDegrees=45
  swing1Motion=Free
  swing2LimitDegrees=45
  swing2Motion=Free
  twistLimitDegrees=45
  twistMotion=Free
 }
 {
  bone1Name=head
  bone2Name=neck_02
  swing1LimitDegrees=45
  swing1Motion=Free
  swing2LimitDegrees=45
  swing2Motion=Free
  twistLimitDegrees=45
  twistMotion=Free
 }
 {
  bone1Name=thigh_l
  bone2Name=pelvis
  swing1LimitDegrees=45
  swing1Motion=Free
  swing2LimitDegrees=45
  swing2Motion=Free
  twistLimitDegrees=45
  twistMotion=Free
 }
 {
  bone1Name=calf_l
  bone2Name=thigh_l
  swing1LimitDegrees=45
  swing1Motion=Free
  swing2LimitDegrees=45
  swing2Motion=Free
  twistLimitDegrees=45
  twistMotion=Free
 }
 {
  bone1Name=foot_l
  bone2Name=calf_l
  swing1LimitDegrees=45
  swing1Motion=Free
  swing2LimitDegrees=45
  swing2Motion=Free
  twistLimitDegrees=45
  twistMotion=Free
 }
 {
  bone1Name=clavicle_r
  bone2Name=spine_05
  swing1LimitDegrees=45
  swing1Motion=Free
  swing2LimitDegrees=45
  swing2Motion=Free
  twistLimitDegrees=45
  twistMotion=Free
 }
 {
  bone1Name=upperarm_r
  bone2Name=clavicle_r
  swing1LimitDegrees=45
  swing1Motion=Free
  swing2LimitDegrees=45
  swing2Motion=Free
  twistLimitDegrees=45
  twistMotion=Free
 }
 {
  bone1Name=lowerarm_r
  bone2Name=upperarm_r
  swing1LimitDegrees=45
  swing1Motion=Free
  swing2LimitDegrees=45
  swing2Motion=Free
  twistLimitDegrees=45
  twistMotion=Free
 }
 {
  bone1Name=foot_r
  bone2Name=calf_r
  swing1LimitDegrees=45
  swing1Motion=Free
  swing2LimitDegrees=45
  swing2Motion=Free
  twistLimitDegrees=45
  twistMotion=Free
 }
 {
  bone1Name=lowerarm_l
  bone2Name=upperarm_l
  swing1LimitDegrees=45
  swing1Motion=Free
  swing2LimitDegrees=45
  swing2Motion=Free
  twistLimitDegrees=45
  twistMotion=Free
 }
 {
  bone1Name=upperarm_l
  bone2Name=clavicle_l
  swing1LimitDegrees=45
  swing1Motion=Free
  swing2LimitDegrees=45
  swing2Motion=Free
  twistLimitDegrees=45
  twistMotion=Free
 }
 {
  bone1Name=clavicle_l
  bone2Name=spine_05
  swing1LimitDegrees=45
  swing1Motion=Free
  swing2LimitDegrees=45
  swing2Motion=Free
  twistLimitDegrees=45
  twistMotion=Free
 }
 {
  bone1Name=calf_r
  bone2Name=thigh_r
  swing1LimitDegrees=45
  swing1Motion=Free
  swing2LimitDegrees=45
  swing2Motion=Free
  twistLimitDegrees=45
  twistMotion=Free
 }
 {
  bone1Name=thigh_r
  bone2Name=pelvis
  swing1LimitDegrees=45
  swing1Motion=Free
  swing2LimitDegrees=45
  swing2Motion=Free
  twistLimitDegrees=45
  twistMotion=Free
 }
]
